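# Translucent UI material for the fall damage icon in the spectator HUD's kill feed.
# It tints the game phase background texture pinkish-red and lerps it toward the fall damage icon, using the icon's own RGB as the per-channel blend alpha.
# The icon's UVs are scaled about their center by a scalar parameter.
# A separate sum combines the icon's alpha with a slightly reduced background alpha.
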Material /Game/Athena/HUD/Spectator/Art/M_KillFeed_FallDamage
graph {
 properties {blendMode=BLEND_Translucent, materialDomain=MD_UI, pixelDepthOffsetMode=PDOM_Legacy, usedShadingModels=Default Lit}
 nodes [
  n1 = TextureSample(UVs=MaterialFunctionCall(UVs=TextureCoordinate(), Texture Scale=ScalarParameter() {defaultValue=0.9}) {materialFunction=&ScaleUVsByCenter}.UVs) {texture=&/KillFeed_FallDamage}
  n2 = TextureSample() {texture=/Game/Athena/HUD/Phase/T_GamePhaseBackground}
  Add(A=@n1.A, B=Multiply(A=@n2.A) {constB=0.9})
  LinearInterpolate(A=Multiply(A=Constant3Vector() {constant=<rgba 1 0 0.1529 1>}, B=@n2.RGB), B=@n1.RGB, Alpha=@n1.RGB)
 ]
}
